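# Translucent unlit drool material for Niagara mesh particles, from the Parcel Petal pickaxe FX.
# A panning distortion normal map perturbs the vertex normal for a view-space reflection lookup, added over a Fresnel two-color blend with optional particle tint.
# The same noise modulates opacity, scaled by particle alpha.
# Dynamic parameters feed a UV-masked rotate-about-axis bend around the mesh origin, with a previous-frame branch for motion vectors.
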Material /Game/Weapons/FORT_Melee/Pickaxe_Parcel_Petal_Female/FX/M_Drool
graph {
 properties {
  bUsedWithNiagaraMeshParticles=true
  blendMode=BLEND_Translucent
  pixelDepthOffsetMode=PDOM_Legacy
  shadingModel=MSM_Unlit
  usedShadingModels=Unlit
 }
 nodes [
  n1 = ScalarParameter() {defaultValue=1.3, parameterName=FresnelMult}
  n2 = ScalarParameter() {defaultValue=2, parameterName=FresnelExp}
  n3 = TextureCoordinate()
  n4 = ScalarParameter() {defaultValue=0.6, parameterName=DistTile}
  n5 = Time() {bOverride_Period=true, period=64}
  n6 = DynamicParameter() {paramNames [rand 0-1, WPO_Amnt, Param3, Param4]}
  n7 = ScalarParameter() {defaultValue=2, parameterName=DistSpeed}
  n8 = Constant() {r=0.1}
  n9 = TextureSample(UVs=Panner(Coordinate=Multiply(A=@n3, B=@n4), Time=Add(A=@n5, B=@n6.rand 0-1), Speed=AppendVector(A=@n7, B=@n8)) {speedX=-1, speedY=0.1}) {samplerType=SAMPLERTYPE_Normal, texture=&T_Weap_Distort_N_02}
  n10 = Multiply(A=Transform(None=@n9.RGB), B=ScalarParameter() {defaultValue=0.5, parameterName=NoiseSTr})
  n11 = VertexNormalWS()
  n12 = Multiply(A=Transform(None=Add(A=@n10, B=@n11)) {transformSourceType=TRANSFORMSOURCE_World, transformType=TRANSFORM_View}) {constB=0.5}
  n13 = TextureSample(UVs=ComponentMask(None=Add(A=@n12) {constB=0.5}) {g=true, r=true}) {texture=/Game/Packages/Fortress_SharedMaterials/Textures/T_Ravens_SpecularReflection}
  n14 = ScalarParameter() {parameterName=ReflectionContrast}
  n15 = Add(
   A=LinearInterpolate(
    A=VectorParameter() {parameterName=ColorA, defaultValue=<rgba 0.8466 0.8466 0.8466 1>}.RGB
    B=VectorParameter() {parameterName=ColorB, defaultValue=<rgba 0.0742 0.0784 0.0839 1>}.RGB
    Alpha=MaterialFunctionCall(Brightness=@n1, Exponent=@n2) {materialFunction=&MF_Fresnel_FX}.Fresnel
   )
   B=Multiply(
    A=MaterialFunctionCall(In=@n13.G, Contrast=@n14) {materialFunction=&CheapContrast}.Result
    B=VectorParameter() {parameterName=ReflectionColor, defaultValue=<rgba 3 3 3 1>}.RGB
   )
  )
  n16 = ParticleColor()
  MaterialFunctionCall(Emissive=StaticSwitchParameter(True=Multiply(A=@n15, B=@n16.RGB), False=@n15) {parameterName=Multiply Particle Color}) {materialFunction=&MF_FX_EmissiveAdjustments}
  n17 = ParticlePositionWS()
  n18 = DynamicParameter() {parameterIndex=3, defaultValue=<rgba 1 1 1 0.15>, paramNames [PosX,PosY,PosZ,Rigidity]}
  n19 = MaterialFunctionCall(R=@n18.PosX, G=@n18.PosX, B=@n18.PosX) {materialFunction=&AppendMany}
  MaterialFunctionCall!(
   Vector3=Multiply(A=ScalarParameter() {defaultValue=1, parameterName=Param_2}, B=Subtract(A=@n17, B=@n19.RGB))
   MaximumNumberOfDigits
   UVs
   DebugTextLocation RG_UpperRight BA_LowerLeft
   Component Spacing
  ) {materialFunction=/Engine/Functions/Engine_MaterialFunctions02/Utility/DebugFloat3Values}
  n20 = Add(A=Multiply(A=@n9.G) {constB=0.5}) {constB=0.5}
  n21 = ScalarParameter() {defaultValue=1, parameterName=contrast}
  n22 = Saturate(None=Add(
    A=ScalarParameter() {defaultValue=0.75, parameterName=OpacBase}
    B=MaterialFunctionCall(In=@n20, Contrast=@n21) {materialFunction=&CheapContrast}.Result
   ))
  n23 = ScalarParameter() {defaultValue=0.95, parameterName=OpacEnd}
  Multiply(A=Multiply(A=Multiply(A=@n22, B=@n23), B=ParticleColor().A))
  n24 = DynamicParameter() {parameterIndex=1, defaultValue=<rgba 0 0 -1 1>, paramNames [RestNormalX, RestNormalY, RestNormalZ, --------------]}
  n25 = Multiply(A=MaterialFunctionCall(R=@n24.RestNormalX, G=@n24.RestNormalX, B=@n24.RestNormalX) {materialFunction=&AppendMany}.RGB) {constB=-1}
  n26 = DynamicParameter() {parameterIndex=2, paramNames [MeshOriginX,MeshOriginY,MeshOriginZ,MaskFalloff]}
  n27 = MaterialFunctionCall(R=@n26.MeshOriginX, G=@n26.MeshOriginX, B=@n26.MeshOriginX, A=@n26.MeshOriginX) {materialFunction=&AppendMany}
  n28 = CrossProduct(A=@n25, B=Normalize(VectorInput=Subtract(A=@n19.RGB, B=@n27.RGB)))
  n29 = WorldPosition()
  n30 = Subtract(A=ComponentMask(None=TextureCoordinate()) {g=true}, B=ScalarParameter() {parameterName=Mask Offset})
  n31 = Saturate(None=Power(Base=Saturate(None=@n30), Exp=@n26.MaskFalloff))
  n32 = DynamicParameter() {paramNames [PrevPosX,PrevPosY,PrevPosZ,GlowAmt]}
  n33 = MaterialFunctionCall(R=@n32.PrevPosX, G=@n32.PrevPosX, B=@n32.PrevPosX) {materialFunction=&AppendMany}
  n34 = CrossProduct(A=@n25, B=Normalize(VectorInput=Subtract(A=@n33.RGB, B=@n27.RGB)))
  PreviousFrameSwitch(
   Current Frame=LinearInterpolate(A=RotateAboutAxis(NormalizedRotationAxis=@n28, RotationAngle=@n18.Rigidity, PivotPoint=@n27.RGB, Position=@n29.XYZ), Alpha=@n31) {constB=0}
   Previous Frame=LinearInterpolate(A=RotateAboutAxis(NormalizedRotationAxis=@n34, RotationAngle=@n18.Rigidity, PivotPoint=@n27.RGB, Position=@n29.XYZ), Alpha=@n31) {constB=0}
  )
  ScalarParameter() {defaultValue=1, parameterName=Falloff}
 ]
}
